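# Niagara idle effect for the PixieParts pickaxe, with a single Glow sprite emitter that changes by cosmetic variant.
# A burst spawns locally aligned sprites at an offset from a socket, and they can persist until the emitter deactivates.
# The system variant indexes a render bool and alpha switch, while a variant color tints the sprite.
# Alpha fades in over system age, and real time drives a dynamic material parameter.
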NiagaraSystem /BRCosmetics/Weapons/FORT_Melee/Pickaxe_PixieParts_Lava/FX/NS_Pickaxe_PixieParts_Idle
dependencies {
 usedDataInterfaces [
  /Script/Niagara.NiagaraDataInterfaceArrayBool
  &NiagaraDataInterfaceCurve
  &NiagaraDataInterfaceSocketReader
  &NiagaraDataInterfaceArrayColor
 ]
 usedDynamicInputs [&MakeIntFromFloat, &SelectLinearColorFromArray]
 usedModules [
  &self:SystemSpawnScript.SystemScriptSource.SystemScriptGraph.NiagaraNodeAssignment_0.SetVariables_#1
  &SystemState
  &self:SystemSpawnScript.SystemScriptSource.SystemScriptGraph.NiagaraNodeAssignment_2.SetVariables_#2
  &self:SystemSpawnScript.SystemScriptSource.SystemScriptGraph.NiagaraNodeAssignment_3.SetVariables_#3
 ]
}
emitters {Glow {
  emitterUpdateScript [
   &EmitterState(Life Cycle Mode=::NewEnumerator1, Loop Behavior=::NewEnumerator1, Loop Duration Mode=::NewEnumerator1, UseLoopDelay=false)
   &SpawnBurst_Instantaneous(Use Spawn Probability=false, UseLoopCountLimit=false)
   /SetVariables(Emitter.RenderBool:NiagaraBool=/Niagara/DynamicInputs/Arrays/SelectBoolFromArray(
     Array Sampling Mode:ENiagara_LifetimeMode=::NewEnumerator0
     Direct Array Index:NiagaraInt32=$System.Variant
     Direct Array Mode:ENiagara_WrapClamp=::NewEnumerator0
     Evaluation Type:ENiagaraRandomnessEvaluation=::NewEnumerator0
     Recalculate Random Each Loop:NiagaraBool=false
     Randomness Mode:ENiagaraRandomnessMode=::NewEnumerator0
     Random Seed:NiagaraInt32=?
     Fixed Random Seed:NiagaraBool=false
     Bool Selection Array:NiagaraDataInterfaceArrayBool=DataInterface() {propertyValues {GpuSyncMode=SyncCpuToGpu, MaxElements=0, BoolData [true,true,false,false,true,false]}}
    )) {isSetParameters=true}
  ]
  particleSpawnScript [
   &InitializeParticle(Color Mode=::NewEnumerator0, Sprite Size=<100 35>, Sprite Size Mode=::NewEnumerator4, UsePositionOffset=false, Write Material Random=true)
   &SocketLocation(
    SamplePosOffset=true
    Sampled Position Offset=<2.9 0 30>
    Sampled Position Offset Coordinate Space:ENiagaraSkelMeshTransforms=::NewEnumerator2
    Socket Reader=$System.Socket Reader
    Socket Sampling Mode=::NewEnumerator2
    VelocitySpeedLimit=false
   )
   &SpriteFacingAndAlignment(
    Alignment Coordinate Space=::Local
    Facing Coordinate Space=::Local
    Sprite Alignment=<1 0 0>
    Sprite Facing=<0 1 0>
    WriteAlignment=true
    WriteFacing=true
   )
  ]
  particleUpdateScript [
   &ParticleState(Kill Particles When Lifetime Has Elapsed=false, Let Infinitely Lived Particles Die When Emitter Deactivates=true)
   &ScaleColor(
    Scale Alpha=&FloatFromCurve(
     FloatCurve=DataInterface() {propertyValues {Curve {
        defaultValue=1
        keys [{interpMode=RCIM_Cubic, tangentMode=RCTM_User, time=0, value=0}, {interpMode=RCIM_Cubic, tangentMode=RCTM_User, time=0.33, value=1}]
       }}}
     CurveIndex=$Engine.System.Age
     Scale Curve=&MergeHelper_FloatSwitch(0=1, 1=1, 2=0, 3=0, 4=1, 5=0, Index=$System.Variant, Num=::NewEnumerator4)
    )
    Scale RGB=&MakeVectorFromLinearColorRGB(Color=$System.VariantColor)
    ScaleA=true
    ScaleRGB=true
   )
   &DynamicMaterialParameters(
    Index 0 Param 1=&Multiply_Float(A=$Engine.RealTime, B=0.4)
    Index 0 Param 2=6
    Module.Index0Float0Write=true
    Module.Index0Float1Write=true
    Module.Index0Float2Write=true
    Module.Index0Float3Write=true
   )
  ]
  rendererClasses [&NiagaraSpriteRendererProperties]
  renderers [{rendererClass=&NiagaraSpriteRendererProperties}]
 }}
summary {
 systemName=NS_Pickaxe_PixieParts_Idle
 emitters [{emitterName=Glow, rendererClasses [&NiagaraSpriteRendererProperties]}]
 userVariables [
  User.bIsFrontEnd {type=NiagaraBool}
  User.bIsFrontEndPreview {type=NiagaraBool}
  User.Variant {type=NiagaraFloat, defaultValue=0, description=""}
 ]
}
