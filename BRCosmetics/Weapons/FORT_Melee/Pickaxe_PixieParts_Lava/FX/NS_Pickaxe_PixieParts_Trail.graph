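# Niagara system for the PixieParts pickaxe's swing trail, with six color variants.
# Its single MainTrail emitter uses anim-trail modules between two bones or sockets to feed three ribbon renderers.
# A System.Variant index switches color curves for the main and dark trail colors, plus an afterimage color that only one variant overrides.
# Alpha curves fade the short-lived particles, and system age also quickly fades the main color's alpha.
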NiagaraSystem /BRCosmetics/Weapons/FORT_Melee/Pickaxe_PixieParts_Lava/FX/NS_Pickaxe_PixieParts_Trail
dependencies {
 usedDataInterfaces [
  &NiagaraDataInterfaceSkeletalMesh
  &NiagaraDataInterfaceColorCurve
  &NiagaraDataInterfaceCurve
  &NiagaraDataInterfaceSocketReader
 ]
 usedDynamicInputs [&MakeIntFromFloat]
 usedModules [
  &self:SystemSpawnScript.SystemScriptSource.SystemScriptGraph.NiagaraNodeAssignment_0.SetVariables_#1
  &SystemState
  &self:SystemSpawnScript.SystemScriptSource.SystemScriptGraph.NiagaraNodeAssignment_1.SetVariables_#2
 ]
}
emitters {MainTrail {
  emitterUpdateScript [
   &EmitterState(Life Cycle Mode=::NewEnumerator1, Loop Behavior=::NewEnumerator1, Loop Duration=1, UseLoopDelay=false)
   &FN_SpawnAnimTrail()
  ]
  particleSpawnScript [
   &FN_InitializeAnimTrailBoneIndecies(First Bone or Socket Index=1, Second Bone or Socket Index=0)
   &FN_InitializeAnimTrail(Lifetime=0.2)
  ]
  particleUpdateScript [
   &FN_UpdateAnimTrail() {module=FN_UpdateAnimTrail001}
   &ScaleColor(
    Scale Alpha=&FloatFromCurve(
     FloatCurve=DataInterface() {propertyValues {Curve {keys [
         {arriveTangent=-0.4286, interpMode=RCIM_Cubic, leaveTangent=-0.4286, tangentMode=RCTM_User, time=0, value=1}
         {interpMode=RCIM_Cubic, tangentMode=RCTM_User, time=1, value=0}
        ]}}}
     Scale Curve=&FloatFromCurve(
      FloatCurve=DataInterface() {propertyValues {Curve {
         defaultValue=1
         keys [{interpMode=RCIM_Cubic, tangentMode=RCTM_User, time=0, value=1}, {interpMode=RCIM_Cubic, tangentMode=RCTM_User, time=0.2, value=0}]
        }}}
      CurveIndex=$Engine.System.Age
     )
    )
    Scale RGB=&MergeHelper_VectorSwitch(
     0=&MakeVectorFromLinearColorRGB(Color=&ColorFromCurve(ColorCurve=DataInterface() {propertyValues {
         AlphaCurve {keys [<at 0 1 RCIM_Cubic>,<at 1 0 RCIM_Cubic>]}
         BlueCurve {keys [<at 0 0>,<at 0.1725 0 RCIM_Cubic>,<at 0.7 0 RCIM_Cubic>]}
         GreenCurve {keys [<at 0 0.2213>, {interpMode=RCIM_Cubic, leaveTangent=-0.3181, time=0.1725, value=0.0604}, <at 0.7 0.0137 RCIM_Cubic>]}
         RedCurve {keys [<at 0 11>, {interpMode=RCIM_Cubic, leaveTangent=-16.4577, time=0.1725, value=3}, <at 0.7 0.2604 RCIM_Cubic>]}
        }}))
     1=&MakeVectorFromLinearColorRGB(Color=&ColorFromCurve(ColorCurve=DataInterface() {propertyValues {
         AlphaCurve {keys [<at 0 1 RCIM_Cubic>,<at 1 0 RCIM_Cubic>]}
         BlueCurve {keys [<at 0 7>, {interpMode=RCIM_Cubic, leaveTangent=-9.1946, time=0.1725, value=2}, <at 0.7 1 RCIM_Cubic>]}
         GreenCurve {keys [<at 0 6.4427>, {interpMode=RCIM_Cubic, leaveTangent=-9.468, time=0.1725, value=1.4386}, <at 0.7 0.2643 RCIM_Cubic>]}
         RedCurve {keys [<at 0 0>, {interpMode=RCIM_Cubic, leaveTangent=0.0001, time=0.1725, value=0.0558}, <at 0.7 0.0001 RCIM_Cubic>]}
        }}))
     2=&MakeVectorFromLinearColorRGB(Color=&ColorFromCurve(ColorCurve=DataInterface() {propertyValues {
         AlphaCurve {keys [<at 0 1 RCIM_Cubic>,<at 1 0 RCIM_Cubic>]}
         BlueCurve {keys [<at 0 0>,<at 0.1725 0 RCIM_Cubic>,<at 0.7 0 RCIM_Cubic>]}
         GreenCurve {keys [<at 0 0.2213>, {arriveTangent=-0.2969, interpMode=RCIM_Cubic, leaveTangent=-0.2969, time=0.1725, value=0.0604}, <at 0.7 0.0134 RCIM_Cubic>]}
         RedCurve {keys [<at 0 11>, {arriveTangent=-15.3423, interpMode=RCIM_Cubic, leaveTangent=-15.3423, time=0.1725, value=3}, <at 0.7 0.2604 RCIM_Cubic>]}
        }}))
     3=&MakeVectorFromLinearColorRGB(Color=&ColorFromCurve(ColorCurve=DataInterface() {propertyValues {
         AlphaCurve {keys [<at 0 1 RCIM_Cubic>,<at 1 0 RCIM_Cubic>]}
         BlueCurve {keys [<at 0 0>,<at 0.1725 0 RCIM_Cubic>,<at 0.7 0 RCIM_Cubic>]}
         GreenCurve {keys [<at 0 1.9096>, {interpMode=RCIM_Cubic, leaveTangent=-2.2727, time=0.1725, value=0.9656}, <at 0.7 0.4266 RCIM_Cubic>]}
         RedCurve {keys [<at 0 7>, {interpMode=RCIM_Cubic, leaveTangent=-9.1946, time=0.1725, value=3}, <at 0.7 1 RCIM_Cubic>]}
        }}))
     4=&MakeVectorFromLinearColorRGB(Color=&ColorFromCurve(ColorCurve=DataInterface() {propertyValues {
         AlphaCurve {keys [<at 0 1 RCIM_Cubic>,<at 1 0 RCIM_Cubic>]}
         BlueCurve {keys [<at 0.0078 7>,<at 0.7 2>]}
         GreenCurve {keys [<at 0.0078 5.6861>,<at 0.7 0.3191>]}
         RedCurve {keys [<at 0.0078 0>,<at 0.7 0>]}
        }}))
     5=&MakeVectorFromLinearColorRGB(Color=&ColorFromCurve(ColorCurve=DataInterface() {propertyValues {
         AlphaCurve {keys [<at 0 1 RCIM_Cubic>,<at 1 0 RCIM_Cubic>]}
         BlueCurve {keys [<at 0 0>,<at 0.1725 0 RCIM_Cubic>,<at 0.7 0 RCIM_Cubic>]}
         GreenCurve {keys [<at 0 0.1408>, {interpMode=RCIM_Cubic, leaveTangent=-0.1948, time=0.1725, value=0.0604}, <at 0.7 0.0137 RCIM_Cubic>]}
         RedCurve {keys [<at 0 7>, {interpMode=RCIM_Cubic, leaveTangent=-10.328, time=0.1725, value=3}, <at 0.7 0.2604 RCIM_Cubic>]}
        }}))
     Index=$System.Variant
     Num=::NewEnumerator4
    )
    ScaleA=true
    ScaleRGB=true
   ) {module=ScaleColor001}
   /SetVariables(Particles.AfterImage:LinearColor=&MakeLinearColorFromVectorAndFloat(
     "Vector (RGB)"=&MergeHelper_VectorSwitch(
      0=&MakeVectorFromLinearColorRGB(Color=$Particles.Color)
      1=&MakeVectorFromLinearColorRGB(Color=$Particles.Color)
      2=&MakeVectorFromLinearColorRGB(Color=$Particles.Color)
      3=&MakeVectorFromLinearColorRGB(Color=$Particles.Color)
      4=&MakeVectorFromLinearColorRGB(Color=&ColorFromCurve(ColorCurve=DataInterface() {propertyValues {
          AlphaCurve {keys [<at 0 1 RCIM_Cubic>,<at 1 0 RCIM_Cubic>]}
          BlueCurve {keys [<at 0 0>,<at 0.1725 0 RCIM_Cubic>,<at 0.6526 0 RCIM_Cubic>]}
          GreenCurve {keys [
            <at 0 4.4718>
            {arriveTangent=-6.1991, interpMode=RCIM_Cubic, leaveTangent=-6.1991, time=0.1725, value=0.9656}
            <at 0.6526 0.4266 RCIM_Cubic>
           ]}
          RedCurve {keys [<at 0 7>, {arriveTangent=-9.1946, interpMode=RCIM_Cubic, leaveTangent=-9.1946, time=0.1725, value=3}, <at 0.6526 1 RCIM_Cubic>]}
         }}))
      5=&MakeVectorFromLinearColorRGB(Color=$Particles.Color)
      Index=$System.Variant
      Num=::NewEnumerator4
     )
     "Float (Alpha)"=&FloatFromCurve(
      FloatCurve=DataInterface() {propertyValues {Curve {
         defaultValue=1
         keys [{interpMode=RCIM_Cubic, tangentMode=RCTM_User, time=0, value=1}, {interpMode=RCIM_Cubic, tangentMode=RCTM_User, time=0.5, value=0}]
        }}}
      Scale Curve=0.25
     )
    )) {isSetParameters=true}
   /SetVariables(Particles.DarkTrailColor:LinearColor=&MakeLinearColorFromVectorAndFloat(
     "Vector (RGB)"=&MakeVectorFromLinearColorRGB(Color=&MergeHelper_ColorSwitch(
       0=<rgba 0.1354 0 0 5>
       1=<rgba 0 0.0501 0.1354 5>
       2=<rgba 0 0 0 5>
       3=<rgba 0.2552 0.1238 0 5>
       4=<rgba 0.2969 0.2371 0.0232 5>
       5=<rgba 0 0 0 5>
       Index=$System.Variant
       Num=::NewEnumerator4
      ))
     "Float (Alpha)"=&FloatFromCurve(
      FloatCurve=DataInterface() {propertyValues {Curve {
         defaultValue=1
         keys [{interpMode=RCIM_Cubic, tangentMode=RCTM_User, time=0, value=1}, {interpMode=RCIM_Cubic, tangentMode=RCTM_User, time=0.5, value=0}]
        }}}
      Scale Curve=3
     )
    )) {isSetParameters=true}
  ]
  rendererClasses [&NiagaraRibbonRendererProperties]
  renderers [
   {rendererClass=&NiagaraRibbonRendererProperties}
   {rendererClass=&NiagaraRibbonRendererProperties, rendererIndex=1}
   {rendererClass=&NiagaraRibbonRendererProperties, rendererIndex=2}
  ]
 }}
summary {
 systemName=NS_Pickaxe_PixieParts_Trail
 emitters [{emitterName=MainTrail, rendererClasses [&NiagaraRibbonRendererProperties]}]
 userVariables [
  User.RibbonIDOffset {type=NiagaraInt32}
  User.Variant {type=NiagaraFloat, defaultValue=0, description=""}
  User.VariantPicker {type=NiagaraFloat, defaultValue=0, description=""}
 ]
}
