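# Niagara impact effect for the PixieParts Lava pickaxe, a short burst of flare, streak and spark-like sprites along the system's direction vector.
# One emitter, Flash_Flare, fires five burst spawn groups whose index drives per-group size, motion, curves and a VisibilityTag for its seven sprite renderers.
# VariantInt switches the color palettes.
# Front-end modifiers shrink sizes and speeds, and platform checks set the count of the dragged, speed-stretched particles.
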NiagaraSystem /BRCosmetics/Weapons/FORT_Melee/Pickaxe_PixieParts_Lava/FX/NS_Pickaxe_Pixieparts_Impact
dependencies {
 usedDataInterfaces [
  &NiagaraDataInterfacePlatformSet
  &NiagaraDataInterfaceCurve
  &NiagaraDataInterfaceVector2DCurve
  &NiagaraDataInterfaceColorCurve
  &NiagaraDataInterfaceArrayFloat3
 ]
 usedDynamicInputs [/Game/Effects/Niagara/Modules/Utility/FN_DirectionVectorByPickaxeType]
 usedModules [
  &self:SystemSpawnScript.SystemScriptSource.SystemScriptGraph.NiagaraNodeAssignment_1.SetVariables_#1
  &self:SystemSpawnScript.SystemScriptSource.SystemScriptGraph.NiagaraNodeAssignment_0.SetVariables_#2
  &self:SystemSpawnScript.SystemScriptSource.SystemScriptGraph.NiagaraNodeAssignment_2.SetVariables_#3
  &SystemState
  &self:SystemSpawnScript.SystemScriptSource.SystemScriptGraph.NiagaraNodeAssignment_3.SetVariables_#4
 ]
}
emitters {Flash_Flare {
  emitterUpdateScript [
   &EmitterState()
   &SpawnBurst_Instantaneous(Spawn Group=4, Use Spawn Probability=false, UseLoopCountLimit=false) {module=SpawnBurst_Instantaneous004}
   &SpawnBurst_Instantaneous(Use Spawn Probability=false, UseLoopCountLimit=false)
   &SpawnBurst_Instantaneous(
    Spawn Count=&RandomRangeInt(Minimum=3, Maximum=4)
    Spawn Group=1
    Use Spawn Probability=false
    UseLoopCountLimit=false
   ) {module=SpawnBurst_Instantaneous001}
   &SpawnBurst_Instantaneous(
    Spawn Count=&MakeIntFromFloat(Float=&FN_PickFloatDependOnPlatform(
      TargetPlatform=DataInterface() {propertyValues {Platforms {qualityLevelMask=-30}}}
      TargetPlatformActive=&FN_PickFloatDependOnPlatform(
       TargetPlatform=DataInterface() {propertyValues {Platforms {qualityLevelMask=-31}}}
       TargetPlatformActive=&MakeFloatFromInt(Int=8)
       TargetPlatformNoActive=&MakeFloatFromInt(Int=10)
      )
      TargetPlatformNoActive=&MakeFloatFromInt(Int=20)
     ))
    Spawn Group=2
    Use Spawn Probability=false
    UseLoopCountLimit=false
   ) {module=SpawnBurst_Instantaneous002}
   &SpawnBurst_Instantaneous(Spawn Count=6, Spawn Group=3, Use Spawn Probability=false, UseLoopCountLimit=false) {module=SpawnBurst_Instantaneous003}
   /SetVariables(Emitter.HueShiftSpike:LinearColor=&MergeHelper_ColorSwitch(
     0=<rgba 0.4609 0.0204 1 1>
     1=<rgba 0 0 0 0>
     2=<rgba 2 0.0417 0 1>
     3=<rgba 0 0 0 0>
     4=<rgba 0 0 0 0>
     5=<rgba 0 0 0 0>
     Index=$System.VariantInt
     Num=::NewEnumerator4
    )) {isSetParameters=true}
  ]
  particleSpawnScript [
   &InitializeParticle(
    Color=&MergeHelper_ColorSwitch(
     0=<rgba 1 1 1 1>
     1=<rgba 3 3 3 0.4>
     2=&MergeHelper_ColorSwitch(
      0=<rgba 1 0.2411 0.2188 0.5>
      1=<rgba 0 0.7996 1 0.5>
      2=<rgba 1 0.5162 0 0.5>
      3=<rgba 1 0.4726 0 0.5>
      4=&RandomRangeLinearColor(Minimum=<rgba 1 0.7758 0 0.5>, Maximum=<rgba 0 0.4728 1 0.5>)
      5=<rgba 1 0.2411 0.2188 0.5>
      Index=$System.VariantInt
      Num=::NewEnumerator4
     )
     3=<rgba 1 1 1 1>
     4=<rgba 1 1 1 1>
     Index=$Emitter.SpawnGroup
     Num=::NewEnumerator3
    )
    Lifetime=&MergeHelper_FloatSwitch(
     0=0.18
     1=&RandomRangeFloat(Minimum=0.1, Maximum=0.15)
     2=&RandomRangeFloat(Minimum=0.05, Maximum=0.45)
     3=&RandomRangeFloat(Minimum=0.15, Maximum=0.2)
     4=0.175
     Index=$Emitter.SpawnGroup
     Num=::NewEnumerator3
    )
    Sprite Size=&MergeHelper_Vector2Switch(
     0=&FN_FrontEndVector2Modifier(Front End:Vector2f=<20 162>, Gameplay:Vector2f=<40 400>)
     1=&Multiply_Vector2D(
      A=&UniformRangedVector2D(Minimum=<25 175>, Maximum=<30 200>)
      B=&FN_FrontEndVector2Modifier(Front End:Vector2f=<0.4 0.7>, Gameplay:Vector2f=<1 1>)
     )
     2=&Multiply_Vector2D_ByFloat(
      Float=&Multiply_Float(
       A=&FN_FrontEndFloatModifier(Gameplay=1, Front End=0.7)
       B=&RandomRangeFloat(Minimum=4.5, Maximum=5)
      )
      Vector2D=<1 1.5>
     )
     3=&Multiply_Vector2D_ByFloat(Float=&FN_FrontEndFloatModifier(Gameplay=1, Front End=0.7), Vector2D=<25 100>)
     4=&Multiply_Vector2D_ByFloat(Float=&FN_FrontEndFloatModifier(Gameplay=1, Front End=0.6), Vector2D=<65 220>)
     Index=$Emitter.SpawnGroup
     Num=::NewEnumerator3
    )
    Sprite Size Mode=::NewEnumerator4
    Sprite UV Mode=::NewEnumerator1
    UsePositionOffset=false
    Write Material Random=true
   )
   &ShapeLocation(
    Cylinder Height=&MergeHelper_FloatSwitch(0=0, 1=0, 2=80, 3=0, Index=$Emitter.SpawnGroup, Num=::NewEnumerator2)
    Cylinder Radius=&MergeHelper_FloatSwitch(0=0, 1=0, 2=2, 3=0, Index=$Emitter.SpawnGroup, Num=::NewEnumerator2)
    MaskSpawnGroup=false
    Override Seed=false
    Rotation Axis=$System.DirectionVector
    Rotation Mode=::NewEnumerator1
    Shape Primitive=::NewEnumerator1
    Surface Only=false
    UseOwnerScale=true
   )
   &CameraOffset(Camera Offset Amount=&MergeHelper_FloatSwitch(0=30, 1=25, 2=25, 3=?, Index=$Emitter.SpawnGroup, Num=::NewEnumerator1))
   &DynamicMaterialParameters(
    Index 0 Param 1=0.2
    Index 0 Param 2=1
    Index 3 Param 1=1.5
    Index 3 Param 2=1
    Module.Index0Float0Write=true
    Module.Index0Float1Write=true
    Module.Index0Float2Write=true
    Module.Index0Float3Write=true
    Module.Index3Float0Write=true
    Module.Index3Float1Write=true
    Module.Index3Float2Write=true
    Module.Index3Float3Write=true
    Write Parameter Index 3=::NewEnumerator1
   )
   /SetVariables(Particles.DynamicParameterAlt:Vector4f=&MakeVector4(X=1, Y=1, Z=1, W=1)) {isSetParameters=true}
   &AddVelocity(
    Module.Use Falloff Distance=false
    Override Seed=false
    Velocity Mode=::NewEnumerator1
    Velocity Speed=&Multiply_Float(
     A=&MergeHelper_FloatSwitch(
      0=0
      1=&RandomRangeFloat(Minimum=13, Maximum=25)
      2=&RandomRangeFloat(Minimum=225, Maximum=300)
      3=&RandomRangeFloat(Minimum=200, Maximum=300)
      4=0
      Index=$Emitter.SpawnGroup
      Num=::NewEnumerator3
     )
     B=$System.FrontEndScale
    )
   ) {module=AddVelocity001}
   &AddVelocity(
    Rotation Coordinate Space=::Simulation
    Velocity=$System.DirectionVector
    Velocity Speed Scale=&MergeHelper_FloatSwitch(
     0=&Multiply_Float(A=400, B=$System.FrontEndScale)
     1=&Multiply_Float(A=125, B=$System.FrontEndScale)
     2=&Multiply_Float(A=&RandomRangeFloat(Minimum=300, Maximum=850), B=$System.FrontEndScale)
     3=?
     Index=$Emitter.SpawnGroup
     Num=::NewEnumerator1
    )
   )
   /SetVariables() {isSetParameters=true}
   /SetVariables(Particles.VisibilityTag=$Emitter.SpawnGroup) {isSetParameters=true}
  ]
  particleUpdateScript [
   &ParticleState()
   &AccelerationForce(Acceleration=&MergeHelper_VectorSwitch(0=<0 0 0>, 1=<0 0 0>, 2=<0 0 -500>, 3=?, Index=$Particles.VisibilityTag, Num=::NewEnumerator1))
   &Drag(
    Drag=&MergeHelper_FloatSwitch(
     0=0
     1=0
     2=&RandomRangeFloat(Minimum=5, Maximum=8)
     3=?
     Index=$Particles.VisibilityTag
     Num=::NewEnumerator1
    )
    Rotational Drag:NiagaraFloat=&MergeHelper_FloatSwitch(0=0, 1=0, 2=1, 3=?, Index=$Particles.VisibilityTag, Num=::NewEnumerator1)
    Use Linear Drag=true
    Use Rotational Drag=true
   )
   &SolveForcesAndVelocity(Clamp Velocity=false, Limit Acceleration=false)
   &ScaleSpriteSizeBySpeed(
    Max Scale Factor=&MergeHelper_Vector2Switch(0=<1 1>, 1=<1 1>, 2=<0.5 5>, 3=?, 4=?, Index=$Particles.VisibilityTag, Num=::NewEnumerator1)
    Min Scale Factor=&MergeHelper_Vector2Switch(0=<1 1>, 1=<1 1>, 2=<0.2 0.5>, 3=?, 4=?, Index=$Particles.VisibilityTag, Num=::NewEnumerator1)
   )
   &ScaleSpriteSize(
    Scale Factor=&MergeHelper_Vector2Switch(
     0=&Vector2DFromCurve(
      Vector2Curve=DataInterface() {propertyValues {
        XCurve {keys [{tangentMode=RCTM_User, time=0, value=0}, <at 0.2 1.075>, {tangentMode=RCTM_User, time=1, value=-0.1}]}
        YCurve {keys [{tangentMode=RCTM_User, time=0, value=0}, <at 0.1 1.025>, {tangentMode=RCTM_User, time=1, value=0.8}]}
       }}
      ScaleCurve=<1 1>
     )
     1=&Vector2DFromCurve(
      Vector2Curve=DataInterface() {propertyValues {
        XCurve {
         defaultValue=1
         keys [
          {arriveTangent=4.1785, interpMode=RCIM_Cubic, leaveTangent=4.1785, tangentMode=RCTM_User, time=0, value=0}
          {arriveTangent=1.1179, interpMode=RCIM_Cubic, leaveTangent=1.1179, tangentMode=RCTM_User, time=0.2042, value=0.6228}
          {arriveTangent=0.0224, interpMode=RCIM_Cubic, leaveTangent=0.0224, tangentMode=RCTM_User, time=1, value=1}
         ]
        }
        YCurve {
         defaultValue=0
         keys [
          {arriveTangent=4.1785, interpMode=RCIM_Cubic, leaveTangent=4.1785, tangentMode=RCTM_User, time=0, value=0}
          {arriveTangent=1.1179, interpMode=RCIM_Cubic, leaveTangent=1.1179, tangentMode=RCTM_User, time=0.2042, value=0.6228}
          {arriveTangent=0.0224, interpMode=RCIM_Cubic, leaveTangent=0.0224, tangentMode=RCTM_User, time=1, value=1}
         ]
        }
       }}
      ScaleCurve=<1 1>
     )
     2=&Vector2DFromFloat(Value=&FloatFromCurve(FloatCurve=DataInterface() {propertyValues {Curve {keys [{arriveTangent=-6.725, interpMode=RCIM_Cubic, leaveTangent=-6.725, tangentMode=RCTM_User, time=0.65, value=1}, <at 1 0 RCIM_Cubic>]}}}))
     3=?
     4=?
     Index=$Particles.VisibilityTag
     Num=::NewEnumerator1
    )
    Scale Sprite Size Mode=::NewEnumerator2
   )
   &ScaleColor(
    Scale Alpha=&MergeHelper_FloatSwitch(
     0=&FloatFromCurve(FloatCurve=DataInterface() {propertyValues {Curve {keys [
          {interpMode=RCIM_Cubic, tangentMode=RCTM_User, time=0.5, value=1}
          {arriveTangent=-6.5352, interpMode=RCIM_Cubic, leaveTangent=-6.5352, tangentMode=RCTM_User, time=1, value=0}
         ]}}})
     1=&FloatFromCurve(FloatCurve=DataInterface() {propertyValues {Curve {keys [
          {interpMode=RCIM_Cubic, tangentMode=RCTM_User, time=0.5, value=1}
          {arriveTangent=-6.5352, interpMode=RCIM_Cubic, leaveTangent=-6.5352, tangentMode=RCTM_User, time=1, value=0}
         ]}}})
     2=1
     3=1
     4=&FloatFromCurve(
      FloatCurve=DataInterface() {propertyValues {Curve {
         defaultValue=1
         keys [{interpMode=RCIM_Cubic, tangentMode=RCTM_User, time=0.25, value=1}, {interpMode=RCIM_Cubic, tangentMode=RCTM_User, time=1, value=0}]
        }}}
      Scale Curve=&FN_FrontEndFloatModifier(Gameplay=1, Front End=0.6)
     )
     Index=$Particles.VisibilityTag
     Num=::NewEnumerator3
    )
    Scale RGB=&MergeHelper_VectorSwitch(
     0=&MergeHelper_VectorSwitch(
      0=&MakeVectorFromLinearColorRGB(Color=&ColorFromCurve(ColorCurve=DataInterface() {propertyValues {
          AlphaCurve {keys [<at 0 1 RCIM_Cubic>,<at 1 0 RCIM_Cubic>]}
          BlueCurve {keys [<at 0 1.5543>, {arriveTangent=-1.9096, interpMode=RCIM_Cubic, time=0.2786, value=0.6908}, <at 1 0 RCIM_Cubic>]}
          GreenCurve {keys [<at 0 1.4531>, {arriveTangent=-1.7585, interpMode=RCIM_Cubic, time=0.2786, value=0.6458}, <at 1 0.0218 RCIM_Cubic>]}
          RedCurve {keys [<at 0 9>, {arriveTangent=-10.7819, interpMode=RCIM_Cubic, time=0.2786, value=4}, <at 1 0.224 RCIM_Cubic>]}
         }}))
      1=&MakeVectorFromLinearColorRGB(Color=&ColorFromCurve(ColorCurve=DataInterface() {propertyValues {
          AlphaCurve {keys [<at 0 1 RCIM_Cubic>,<at 1 0 RCIM_Cubic>]}
          BlueCurve {keys [<at 0 9>, {arriveTangent=-8.776, interpMode=RCIM_Cubic, leaveTangent=-8.776, time=0.2786, value=4}, <at 1 0.224 RCIM_Cubic>]}
          GreenCurve {keys [<at 0 8.1379>, {arriveTangent=-8.073, interpMode=RCIM_Cubic, leaveTangent=-8.073, time=0.2786, value=2.4521}, <at 1 0.0649 RCIM_Cubic>]}
          RedCurve {keys [<at 0 2.3686>, {arriveTangent=-2.3325, interpMode=RCIM_Cubic, leaveTangent=-2.3325, time=0.2786, value=0.6062}, <at 1 0.0361 RCIM_Cubic>]}
         }}))
      2=&MakeVectorFromLinearColorRGB(Color=&ColorFromCurve(ColorCurve=DataInterface() {propertyValues {
          AlphaCurve {keys [<at 0 1 RCIM_Cubic>,<at 1 0 RCIM_Cubic>]}
          BlueCurve {keys [<at 0 1.0224>, {arriveTangent=-1.4868, interpMode=RCIM_Cubic, leaveTangent=-1.4868, time=0.2786, value=0.3287}, <at 0.6876 0>]}
          GreenCurve {keys [<at 0 4.2085>, {arriveTangent=-5.9784, interpMode=RCIM_Cubic, leaveTangent=-5.9784, time=0.2786, value=1.0207}, <at 0.6876 0.0974>]}
          RedCurve {keys [<at 0 9>, {arriveTangent=-12.3161, interpMode=RCIM_Cubic, leaveTangent=-12.3161, time=0.2786, value=4}, <at 0.6876 0.5308>]}
         }}))
      3=&MakeVectorFromLinearColorRGB(Color=&ColorFromCurve(ColorCurve=DataInterface() {propertyValues {
          AlphaCurve {keys [<at 0 1 RCIM_Cubic>,<at 1 0 RCIM_Cubic>]}
          BlueCurve {keys [<at 0 0>,<at 0.2786 0 RCIM_Cubic>,<at 1 0 RCIM_Cubic>]}
          GreenCurve {keys [<at 0 4.009>, {arriveTangent=-3.8858, interpMode=RCIM_Cubic, leaveTangent=-3.8858, time=0.2786, value=1.547}, <at 1 0.1232 RCIM_Cubic>]}
          RedCurve {keys [<at 0 9>, {arriveTangent=-8.4583, interpMode=RCIM_Cubic, leaveTangent=-8.4583, time=0.2786, value=4}, <at 1 0.5417 RCIM_Cubic>]}
         }}))
      4=&MakeVectorFromLinearColorRGB(Color=&ColorFromCurve(ColorCurve=DataInterface() {propertyValues {
          AlphaCurve {keys [<at 0 1 RCIM_Cubic>,<at 1 0 RCIM_Cubic>]}
          BlueCurve {keys [{arriveTangent=-16.5307, interpMode=RCIM_Cubic, time=0, value=0}, <at 0.887 0>]}
          GreenCurve {keys [{arriveTangent=-3.6865, interpMode=RCIM_Cubic, time=0, value=1.4888}, <at 0.887 0.5703>]}
          RedCurve {keys [{arriveTangent=0.8064, interpMode=RCIM_Cubic, time=0, value=2}, <at 0.887 0.8021>]}
         }}))
      5=&MakeVectorFromLinearColorRGB(Color=&ColorFromCurve(ColorCurve=DataInterface() {propertyValues {
          AlphaCurve {keys [<at 0 1 RCIM_Cubic>,<at 1 0 RCIM_Cubic>]}
          BlueCurve {keys [<at 0 1.5543>, {arriveTangent=-1.5543, interpMode=RCIM_Cubic, leaveTangent=-1.5543, time=0.2786, value=0.6908}, <at 1 0 RCIM_Cubic>]}
          GreenCurve {keys [<at 0 1.4531>, {arriveTangent=-1.4314, interpMode=RCIM_Cubic, leaveTangent=-1.4314, time=0.2786, value=0.6458}, <at 1 0.0218 RCIM_Cubic>]}
          RedCurve {keys [<at 0 9>, {arriveTangent=-8.776, interpMode=RCIM_Cubic, leaveTangent=-8.776, time=0.2786, value=4}, <at 1 0.224 RCIM_Cubic>]}
         }}))
      Index=$System.VariantInt
      Num=::NewEnumerator4
     )
     1=&MergeHelper_VectorSwitch(
      0=&MakeVectorFromLinearColorRGB(Color=<rgba 1 0.0476 0 1>)
      1=&MakeVectorFromLinearColorRGB(Color=<rgba 0 0.4881 1 1>)
      2=&MakeVectorFromLinearColorRGB(Color=<rgba 1 0.0516 0 1>)
      3=&MakeVectorFromLinearColorRGB(Color=<rgba 1 0.319 0 1>)
      4=&MakeVectorFromLinearColorRGB(Color=<rgba 0 0.1581 1 1>)
      5=&MakeVectorFromLinearColorRGB(Color=<rgba 1 0.1732 0.1302 1>)
      Index=$System.VariantInt
      Num=::NewEnumerator4
     )
     2=&MergeHelper_VectorSwitch(
      0=&MakeVectorFromLinearColorRGB(Color=&ColorFromCurve(ColorCurve=DataInterface() {propertyValues {
          AlphaCurve {keys [<at 0 1 RCIM_Cubic>,<at 1 0 RCIM_Cubic>]}
          BlueCurve {keys [<at 0 3.2804 RCIM_Cubic>,<at 0.6636 0.0023 RCIM_Cubic>]}
          GreenCurve {keys [<at 0 3.2031 RCIM_Cubic>,<at 0.6636 0 RCIM_Cubic>]}
          RedCurve {keys [<at 0 5 RCIM_Cubic>,<at 0.6636 0.2969 RCIM_Cubic>]}
         }}))
      1=&MakeVectorFromLinearColorRGB(Color=&ColorFromCurve(ColorCurve=DataInterface() {propertyValues {
          AlphaCurve {keys [<at 0 1 RCIM_Cubic>,<at 1 0 RCIM_Cubic>]}
          BlueCurve {keys [<at 0 3.2804 RCIM_Cubic>,<at 0.6636 0.2969 RCIM_Cubic>]}
          GreenCurve {keys [<at 0 3.2031 RCIM_Cubic>,<at 0.6636 0.1857 RCIM_Cubic>]}
          RedCurve {keys [<at 0 5 RCIM_Cubic>,<at 0.6636 0.0282 RCIM_Cubic>]}
         }}))
      2=&MakeVectorFromLinearColorRGB(Color=&ColorFromCurve(ColorCurve=DataInterface() {propertyValues {
          AlphaCurve {keys [<at 0 1 RCIM_Cubic>,<at 1 0 RCIM_Cubic>]}
          BlueCurve {keys [<at 0 0.522 RCIM_Cubic>,<at 0.6636 0.0279 RCIM_Cubic>]}
          GreenCurve {keys [<at 0 1.7317 RCIM_Cubic>,<at 0.6636 0.1279 RCIM_Cubic>]}
          RedCurve {keys [<at 0 5 RCIM_Cubic>,<at 0.6636 0.4948 RCIM_Cubic>]}
         }}))
      3=&MakeVectorFromLinearColorRGB(Color=&ColorFromCurve(ColorCurve=DataInterface() {propertyValues {
          AlphaCurve {keys [<at 0 1 RCIM_Cubic>,<at 1 0 RCIM_Cubic>]}
          BlueCurve {keys [<at 0 3.2804 RCIM_Cubic>,<at 0.6636 0 RCIM_Cubic>]}
          GreenCurve {keys [<at 0 3.2031 RCIM_Cubic>,<at 0.6636 0.1253 RCIM_Cubic>]}
          RedCurve {keys [<at 0 5 RCIM_Cubic>,<at 0.6636 0.2969 RCIM_Cubic>]}
         }}))
      4=&MakeVectorFromLinearColorRGB(Color=&ColorFromCurve(ColorCurve=DataInterface() {propertyValues {
          AlphaCurve {keys [<at 0 1 RCIM_Cubic>,<at 1 0 RCIM_Cubic>]}
          BlueCurve {keys [<at 0 5 RCIM_Cubic>,<at 0.6636 0 RCIM_Cubic>]}
          GreenCurve {keys [<at 0 1.0104 RCIM_Cubic>,<at 0.6636 0.5498 RCIM_Cubic>]}
          RedCurve {keys [<at 0 0 RCIM_Cubic>,<at 0.6636 1 RCIM_Cubic>]}
         }}))
      5=&MakeVectorFromLinearColorRGB(Color=&ColorFromCurve(ColorCurve=DataInterface() {propertyValues {
          AlphaCurve {keys [<at 0 1 RCIM_Cubic>,<at 1 0 RCIM_Cubic>]}
          BlueCurve {keys [<at 0 3.2804 RCIM_Cubic>,<at 0.6636 0.0023 RCIM_Cubic>]}
          GreenCurve {keys [<at 0 3.2031 RCIM_Cubic>,<at 0.6636 0 RCIM_Cubic>]}
          RedCurve {keys [<at 0 5 RCIM_Cubic>,<at 0.6636 0.2969 RCIM_Cubic>]}
         }}))
      Index=$System.VariantInt
      Num=::NewEnumerator4
     )
     3=&MakeVectorFromLinearColorRGB(Color=&MergeHelper_ColorSwitch(
       0=<rgba 4 0.5321 0.3958 1>
       1=<rgba 0.814 2.4869 4 1>
       2=&ColorFromCurve(ColorCurve=DataInterface() {propertyValues {
          AlphaCurve {keys [<at 0 1 RCIM_Cubic>,<at 1 0 RCIM_Cubic>]}
          BlueCurve {keys [<at 0 0 RCIM_Cubic>,<at 0.2154 0>,<at 0.8128 0 RCIM_Cubic>]}
          GreenCurve {keys [<at 0 0.6675 RCIM_Cubic>,<at 0.2154 0.6582>,<at 0.8128 0.0424 RCIM_Cubic>]}
          RedCurve {keys [<at 0 2 RCIM_Cubic>,<at 0.2154 1.9878>,<at 0.8128 1.125 RCIM_Cubic>]}
         }})
       3=<rgba 2 0.8167 0 1>
       4=&ColorFromCurve(ColorCurve=DataInterface() {propertyValues {
          AlphaCurve {keys [<at 0 1 RCIM_Cubic>,<at 1 0 RCIM_Cubic>]}
          BlueCurve {keys [<at 0 0 RCIM_Cubic>,<at 0.5972 2 RCIM_Cubic>]}
          GreenCurve {keys [<at 0 1.576 RCIM_Cubic>,<at 0.5972 0.7319 RCIM_Cubic>]}
          RedCurve {keys [<at 0 2.5 RCIM_Cubic>,<at 0.5972 0 RCIM_Cubic>]}
         }})
       5=<rgba 2 0.0391 0 1>
       Index=$System.VariantInt
       Num=::NewEnumerator4
      ))
     4=&MergeHelper_VectorSwitch(
      0=&MakeVectorFromLinearColorRGB(Color=<rgba 0.0833 0 0.0051 2>)
      1=&MakeVectorFromLinearColorRGB(Color=<rgba 0.0246 0.0495 0.0833 2>)
      2=&MakeVectorFromLinearColorRGB(Color=<rgba 0.0833 0.0087 0 2>)
      3=&MakeVectorFromLinearColorRGB(Color=<rgba 0.0833 0.0471 0 2>)
      4=&MakeVectorFromLinearColorRGB(Color=<rgba 0 0.0562 0.1302 2>)
      5=&MakeVectorFromLinearColorRGB(Color=<rgba 0.0833 0 0.0051 2>)
      Index=$System.VariantInt
      Num=::NewEnumerator4
     )
     Index=$Particles.VisibilityTag
     Num=::NewEnumerator3
    )
    ScaleA=true
    ScaleRGB=true
   )
   /SetVariables(
    Particles.GlowColor=&MakeLinearColorFromVectorAndFloat(
     "Vector (RGB)"=&MakeVectorFromLinearColorRGB(Color=&HueShiftLinearColor(Hue=0, Saturation=1.25, Value=1, Linear Color=$Particles.Initial.Color))
     "Float (Alpha)"=&FloatFromCurve(FloatCurve=DataInterface() {propertyValues {Curve {keys [<at 0.5 1>,<at 1 0>]}}}, Scale Curve=0.03)
    )
    Particles.GlowSize=&Multiply_Vector2D(A=$Particles.SpriteSize, B=<8 2>)
   ) {isSetParameters=true}
   /SetVariables(
    Particles.ConeEdge_Color:LinearColor=&Multiply_LinearColorByFloat(Linear Color=$Particles.Color, Float=1)
    Particles.ConeEdge_SpriteSize:Vector2f=&Multiply_Vector2D_ByFloat(
     Float=&Multiply_Float(A=1.2, B=&FN_FrontEndFloatModifier(Gameplay=1, Front End=0.7))
     Vector2D=$Particles.SpriteSize
    )
   ) {isSetParameters=true}
  ]
  rendererClasses [&NiagaraSpriteRendererProperties]
  renderers [
   {rendererClass=&NiagaraSpriteRendererProperties}
   {rendererClass=&NiagaraSpriteRendererProperties, rendererIndex=1}
   {rendererClass=&NiagaraSpriteRendererProperties, rendererIndex=2}
   {rendererClass=&NiagaraSpriteRendererProperties, rendererIndex=3}
   {rendererClass=&NiagaraSpriteRendererProperties, rendererIndex=4}
   {rendererClass=&NiagaraSpriteRendererProperties, rendererIndex=5}
   {rendererClass=&NiagaraSpriteRendererProperties, rendererIndex=6}
  ]
 }}
summary {
 systemName=NS_Pickaxe_Pixieparts_Impact
 emitters [{emitterName=Flash_Flare, rendererClasses [&NiagaraSpriteRendererProperties]}]
 userVariables [
  User.bIsFrontEnd {type=NiagaraBool}
  User.bIsFrontEndPreview {type=NiagaraBool}
  User.SwingSectionIndex {type=NiagaraInt32}
  User.Variant {type=NiagaraFloat, defaultValue=4, description=""}
 ]
}
